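# Niagara system that draws the ribbon swing trail for the HappyFNCS pickaxe cosmetic, using three main trail layers and an Accents layer.
# Each emitter spawns anim-trail particles between a pair of bone or socket indices, with a ribbon ID offset utility at spawn.
# Short-lived ribbons fade over life with time-of-day color and alpha adjustments.
# Main layers scale opacity differently for gameplay and front end, while Accents also fades by emitter age.
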NiagaraSystem /BRCosmetics/Weapons/FORT_Melee/Pickaxe_HappyFNCS/FX/NS_Pickaxe_HappyFNCS_Trail
dependencies {
 usedDataInterfaces [&NiagaraDataInterfaceCurve, &NiagaraDataInterfaceColorCurve, &NiagaraDataInterfaceSocketReader]
 usedDynamicInputs [&MakeCustomLinearColorFromBool]
 usedModules [
  &self:SystemSpawnScript.SystemScriptSource.SystemScriptGraph.NiagaraNodeAssignment_1.SetVariables_#1
  &SystemState
  &self:SystemSpawnScript.SystemScriptSource.SystemScriptGraph.NiagaraNodeAssignment_0.SetVariables_#2
 ]
}
emitters {
 Accents {
  emitterSpawnScript [&FN_RibbonIDOffsetUtility()]
  emitterUpdateScript [
   &EmitterState()
   &FN_SpawnAnimTrail()
  ]
  particleSpawnScript [
   &FN_InitializeAnimTrailBoneIndecies(First Bone or Socket Index=3, Second Bone or Socket Index=4)
   &FN_InitializeAnimTrail(Color=<rgba 0.0467 0.6724 0.7454 1>, Lifetime=0.2)
  ]
  particleUpdateScript [
   &SolveForcesAndVelocity(Clamp Velocity=false, Limit Acceleration=false)
   &FN_UpdateAnimTrail() {module=FN_UpdateAnimTrail001}
   &ScaleColor(
    Scale Alpha=&FloatFromCurve(
     FloatCurve=DataInterface() {propertyValues {Curve {keys [
         {arriveTangent=-2.0466, interpMode=RCIM_Cubic, leaveTangent=-2.0466, tangentMode=RCTM_User, time=0, value=1}
         {interpMode=RCIM_Cubic, tangentMode=RCTM_User, time=1, value=0}
        ]}}}
     Scale Curve=&FloatFromCurve(
      FloatCurve=DataInterface() {propertyValues {Curve {keys [{arriveTangent=-12.7773, interpMode=RCIM_Cubic, leaveTangent=-12.7773, tangentMode=RCTM_User, time=0.1, value=1}, <at 0.25 0 RCIM_Cubic>]}}}
      CurveIndex=$Emitter.Age
      Scale Curve=&FNTimeOfDayScalar(Night Scale=0.5, Morning Scale=0.7, Day Scale=0.7, Evening Scale=0.7)
     )
    )
    Scale RGB=&FN_Select_Vector_from_Time_Of_Day(Morning=<1 1 1>, Day=<1 1 1>, Evening=<1 1 1>, Night=<0.25 0.5 1>)
    ScaleA=true
    ScaleRGB=true
   )
  ]
  rendererClasses [&NiagaraRibbonRendererProperties]
  renderers [
   {rendererClass=&NiagaraRibbonRendererProperties}
   {rendererClass=&NiagaraRibbonRendererProperties, rendererIndex=1}
  ]
 }
 MainTrail {
  emitterSpawnScript [&FN_RibbonIDOffsetUtility()]
  emitterUpdateScript [
   &EmitterState()
   &FN_SpawnAnimTrail()
  ]
  particleSpawnScript [
   &FN_InitializeAnimTrailBoneIndecies(First Bone or Socket Index=1, Second Bone or Socket Index=0)
   &FN_InitializeAnimTrailBoneIndecies(AnimTrail Selection=1, First Bone or Socket Index=2, Second Bone or Socket Index=0) {enabled=false, module=FN_InitializeAnimTrailBoneIndecies001}
   /SetVariables(Particles.RibbonWidth=100) {isSetParameters=true, module=SetVariables_#3}
   &FN_InitializeAnimTrail(
    AnimTrailWidth=&MergeHelper_FloatSwitch(0=1, 1=1, 2=?, 3=?, Index=$Particles.SpawnGroup, Num=::NewEnumerator0)
    Color=$System.MainColor
    Lifetime=&MergeHelper_FloatSwitch(0=0.25, 1=0.1758, 2=?, 3=?, Index=$Particles.SpawnGroup, Num=::NewEnumerator0)
   )
  ]
  particleUpdateScript [
   &FN_UpdateAnimTrail(Locked Particle RibbonWidth:NiagaraFloat=$Particles.RibbonWidth)
   /SetVariables(Particles.Color=&MultiplyLinearColor(
     Linear Color=&MultiplyLinearColor(
      Linear Color=&MergeHelper_ColorSwitch(0=$System.MainColor, 1=<rgba 0.5 0.5 0.5 0.75>, 2=?, 3=?, Index=$Particles.SpawnGroup, Num=::NewEnumerator0)
      Scale Factor=&MakeVector4_FromLinearColor(LinearColor=&ColorFromCurve(ColorCurve=DataInterface() {propertyValues {
          AlphaCurve {keys [<at 0 1 RCIM_Cubic>,<at 1 0 RCIM_Cubic>]}
          BlueCurve {keys [{arriveTangent=-0.7517, interpMode=RCIM_Cubic, time=0, value=1}, <at 1 0.1 RCIM_Cubic>]}
          GreenCurve {keys [{arriveTangent=-0.7517, interpMode=RCIM_Cubic, time=0, value=1}, <at 1 0.1 RCIM_Cubic>]}
          RedCurve {keys [{arriveTangent=-0.7517, interpMode=RCIM_Cubic, time=0, value=1}, <at 1 0.1 RCIM_Cubic>]}
         }}))
     )
     Scale Factor=&MakeVector4_FromLinearColor(LinearColor=&MergeHelper_ColorSwitch(
       0=&FNTimeOfDayColor(Night Color=<rgba 0.0674 0.3328 0.75 1>, Morning Color=<rgba 1 1 1 1>, Day Color=<rgba 1 1 1 1>, Evening Color=<rgba 1 1 1 1>)
       1=$System.TODColor
       2=?
       3=?
       Index=$Particles.SpawnGroup
       Num=::NewEnumerator0
      ))
    )) {isSetParameters=true, module=SetVariables_#4}
   &ScaleColor(
    Color Value To Scale=$StackContext.Color
    Scale Alpha=&Multiply_Float(
     A=&FloatFromCurve(
      FloatCurve=DataInterface() {propertyValues {Curve {keys [
          {interpMode=RCIM_Cubic, leaveTangent=-5.8353, tangentMode=RCTM_Break, time=0, value=1.5}
          {arriveTangent=-0.3285, interpMode=RCIM_Cubic, leaveTangent=-0.3285, tangentMode=RCTM_User, time=0.6, value=0.15}
          {arriveTangent=-0.2139, interpMode=RCIM_Cubic, tangentMode=RCTM_Break, time=1, value=0}
         ]}}}
      Scale Curve=&FN_FrontEndFloatModifier(Gameplay=4, Front End=1)
     )
     B=&FNTimeOfDayScalar(Night Scale=0.5, Morning Scale=0.5, Evening Scale=0.5)
    )
    Scale RGB=&VectorFromFloat(Value=0.7)
    ScaleA=true
    ScaleRGB=true
   )
   /SetVariables(Particles.TipWidth:NiagaraFloat=&FloatFromCurve(FloatCurve=DataInterface() {propertyValues {Curve {keys [<at 0 1>,<at 1 0>]}}}, Scale Curve=10)) {isSetParameters=true, module=SetVariables_#5}
  ]
  rendererClasses [&NiagaraRibbonRendererProperties]
  renderers [{rendererClass=&NiagaraRibbonRendererProperties}]
 }
 MainTrail001 {
  emitterSpawnScript [&FN_RibbonIDOffsetUtility()]
  emitterUpdateScript [
   &EmitterState()
   &FN_SpawnAnimTrail()
  ]
  particleSpawnScript [
   &FN_InitializeAnimTrailBoneIndecies(First Bone or Socket Index=2, Second Bone or Socket Index=0)
   &FN_InitializeAnimTrailBoneIndecies(AnimTrail Selection=1, First Bone or Socket Index=2, Second Bone or Socket Index=0) {enabled=false, module=FN_InitializeAnimTrailBoneIndecies001}
   &FN_InitializeAnimTrail(Lifetime=&MergeHelper_FloatSwitch(0=0.25, 1=0.1758, 2=?, 3=?, Index=$Particles.SpawnGroup, Num=::NewEnumerator0))
   /SetVariables(Particles.RibbonWidth2:NiagaraFloat=&Multiply_Float(A=$Particles.RibbonWidth, B=2)) {isSetParameters=true, module=SetVariables_#3}
  ]
  particleUpdateScript [
   &FN_UpdateAnimTrail(Locked Particle RibbonWidth:NiagaraFloat=$Particles.RibbonWidth)
   /SetVariables(Particles.Color=&FN_Select_Color_from_Time_Of_Day(Morning=<rgba 1 1 1 1>, Day=<rgba 0.5431 0.5525 0.7396 1>, Evening=<rgba 1 1 1 1>, Night=<rgba 1 1 1 1>)) {isSetParameters=true, module=SetVariables_#4}
   &ScaleColor(
    Color Value To Scale=$StackContext.Color
    Scale Alpha=&Multiply_Float(
     A=&FloatFromCurve(
      FloatCurve=DataInterface() {propertyValues {Curve {keys [
          {interpMode=RCIM_Cubic, leaveTangent=-5.8353, tangentMode=RCTM_Break, time=0, value=1.5}
          {arriveTangent=-0.3285, interpMode=RCIM_Cubic, leaveTangent=-0.3285, tangentMode=RCTM_User, time=0.6, value=0.15}
          {arriveTangent=-0.2139, interpMode=RCIM_Cubic, tangentMode=RCTM_Break, time=1, value=0}
         ]}}}
      Scale Curve=&FN_FrontEndFloatModifier(Gameplay=2, Front End=1)
     )
     B=&FNTimeOfDayScalar(Night Scale=1, Morning Scale=0.5, Evening Scale=0.5)
    )
    Scale RGB=&Multiply_Vector(
     A=&VectorFromFloat(Value=0.7)
     B=&FN_Select_Vector_from_Time_Of_Day(Morning=<1 1 1>, Day=<1 1 1>, Evening=<1 1 1>, Night=<0.25 0.5 1>)
    )
    ScaleA=true
    ScaleRGB=true
   )
   /SetVariables(Particles.TipWidth:NiagaraFloat=&FloatFromCurve(FloatCurve=DataInterface() {propertyValues {Curve {keys [<at 0 1>,<at 1 0>]}}}, Scale Curve=10)) {isSetParameters=true, module=SetVariables_#5}
  ]
  rendererClasses [&NiagaraRibbonRendererProperties]
  renderers [
   {rendererClass=&NiagaraRibbonRendererProperties}
   {rendererClass=&NiagaraRibbonRendererProperties, rendererIndex=1}
  ]
 }
 MainTrail002 {
  emitterSpawnScript [&FN_RibbonIDOffsetUtility()]
  emitterUpdateScript [
   &EmitterState()
   &FN_SpawnAnimTrail()
  ]
  particleSpawnScript [
   &FN_InitializeAnimTrailBoneIndecies(First Bone or Socket Index=1, Second Bone or Socket Index=0)
   &FN_InitializeAnimTrailBoneIndecies(AnimTrail Selection=1, First Bone or Socket Index=2, Second Bone or Socket Index=0) {enabled=false, module=FN_InitializeAnimTrailBoneIndecies001}
   /SetVariables(Particles.RibbonWidth=100) {isSetParameters=true, module=SetVariables_#3}
   &FN_InitializeAnimTrail(
    AnimTrailWidth=0.75
    Color=$System.MainColor
    Lifetime=&MergeHelper_FloatSwitch(0=0.25, 1=0.1758, 2=?, 3=?, Index=$Particles.SpawnGroup, Num=::NewEnumerator0)
   )
  ]
  particleUpdateScript [
   &FN_UpdateAnimTrail(Locked Particle RibbonWidth:NiagaraFloat=$Particles.RibbonWidth)
   /SetVariables(Particles.Color=&MultiplyLinearColor(
     Linear Color=&MultiplyLinearColor(
      Linear Color=&MergeHelper_ColorSwitch(0=$System.MainColor, 1=<rgba 0.5 0.5 0.5 0.75>, 2=?, 3=?, Index=$Particles.SpawnGroup, Num=::NewEnumerator0)
      Scale Factor=&MakeVector4_FromLinearColor(LinearColor=&ColorFromCurve(ColorCurve=DataInterface() {propertyValues {
          AlphaCurve {keys [<at 0 1 RCIM_Cubic>,<at 1 0 RCIM_Cubic>]}
          BlueCurve {keys [{arriveTangent=-0.7517, interpMode=RCIM_Cubic, time=0, value=1}, <at 1 0.1 RCIM_Cubic>]}
          GreenCurve {keys [{arriveTangent=-0.7517, interpMode=RCIM_Cubic, time=0, value=1}, <at 1 0.1 RCIM_Cubic>]}
          RedCurve {keys [{arriveTangent=-0.7517, interpMode=RCIM_Cubic, time=0, value=1}, <at 1 0.1 RCIM_Cubic>]}
         }}))
     )
     Scale Factor=&MakeVector4_FromLinearColor(LinearColor=&MergeHelper_ColorSwitch(
       0=&FNTimeOfDayColor(Night Color=<rgba 0.0674 0.3328 0.75 1>, Morning Color=<rgba 1 1 1 1>, Day Color=<rgba 1 1 1 1>, Evening Color=<rgba 1 1 1 1>)
       1=$System.TODColor
       2=?
       3=?
       Index=$Particles.SpawnGroup
       Num=::NewEnumerator0
      ))
    )) {isSetParameters=true, module=SetVariables_#4}
   &ScaleColor(
    Color Value To Scale=$StackContext.Color
    Scale Alpha=&Multiply_Float(
     A=&FloatFromCurve(
      FloatCurve=DataInterface() {propertyValues {Curve {keys [
          {interpMode=RCIM_Cubic, leaveTangent=-5.8353, tangentMode=RCTM_Break, time=0, value=1.5}
          {arriveTangent=-0.3285, interpMode=RCIM_Cubic, leaveTangent=-0.3285, tangentMode=RCTM_User, time=0.6, value=0.15}
          {arriveTangent=-0.2139, interpMode=RCIM_Cubic, tangentMode=RCTM_Break, time=1, value=0}
         ]}}}
      Scale Curve=&FN_FrontEndFloatModifier(Gameplay=5, Front End=1)
     )
     B=&FNTimeOfDayScalar(Night Scale=0.5, Morning Scale=0.5, Evening Scale=0.5)
    )
    Scale RGB=&VectorFromFloat(Value=0.7)
    ScaleA=true
    ScaleRGB=true
   )
   /SetVariables(Particles.TipWidth:NiagaraFloat=&FloatFromCurve(FloatCurve=DataInterface() {propertyValues {Curve {keys [<at 0 1>,<at 1 0>]}}}, Scale Curve=10)) {isSetParameters=true, module=SetVariables_#5}
  ]
  rendererClasses [&NiagaraRibbonRendererProperties]
  renderers [{rendererClass=&NiagaraRibbonRendererProperties}]
 }
}
summary {
 systemName=NS_Pickaxe_HappyFNCS_Trail
 emitters [
  {emitterName=Accents, rendererClasses [&NiagaraRibbonRendererProperties]}
  {emitterName=MainTrail, rendererClasses [&NiagaraRibbonRendererProperties]}
  {emitterName=MainTrail001, rendererClasses [&NiagaraRibbonRendererProperties]}
  {emitterName=MainTrail002, rendererClasses [&NiagaraRibbonRendererProperties]}
 ]
 userVariables [
  User.bIsFrontEnd {type=NiagaraBool}
  User.bIsFrontEndPreview {type=NiagaraBool}
  User.PhysicalSurface {type=NiagaraInt32}
  User.RibbonIDOffset {type=NiagaraInt32}
 ]
}
